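# Unlit translucent, two-sided material for the Carbide glider's Niagara ribbon trails, drawing pulsing, scrolling energy streaks.
# Panned ramp and smoke textures mask a pulse color, which multiplies itself plus a dark-to-bright gradient before MF_FX_EmissiveAdjustments.
# DepthFade intensifies the gradient near intersecting geometry, with a simpler mobile path.
# Opacity combines pulse and static ramp, fades at both trail ends, and scales by particle alpha.
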Material /BRCosmetics/Gadgets/Assets/VinderTech_GliderChute/Glider_CarbideWhite/FX/M_Carbide_Glider_Trails
graph {
 properties {
  bAutomaticallySetUsageInEditor=false
  bUsedWithNiagaraRibbons=true
  bUsedWithStaticMesh=false
  blendMode=BLEND_Translucent
  pixelDepthOffsetMode=PDOM_Legacy
  shadingModel=MSM_Unlit
  twoSided=true
  usedShadingModels=Unlit
 }
 nodes [
  n1 = TextureCoordinate() {coordinateIndex=1}
  n2 = Time()
  n3 = ScalarParameter() {defaultValue=0.02, parameterName=EmissivePulseSpeed}
  n4 = TextureSample(UVs=Panner(
    Coordinate=MaterialFunctionCall(XY=@n1) {materialFunction=&Swizzle}.YX
    Time=Multiply(A=@n2, B=@n3)
   ) {speedY=-1}) {texture=/Game/Global/Textures/Default/Masks_and_Gradients/T_Water_Top_Ramp}
  n5 = TextureCoordinate() {coordinateIndex=1}
  n6 = Power(Base=TextureSample(UVs=Panner(Coordinate=@n5) {speedX=-5}) {samplerType=SAMPLERTYPE_Grayscale, texture=&T_Smoke_Trail_Soft}.G) {constExponent=2.5}
  n7 = LinearInterpolate(
   B=VectorParameter() {desc=16, parameterName=EmissiveColorPulse1, defaultValue=<rgba 0 18.097 30 0>}.RGB
   Alpha=Clamp(None=Add(A=@n4.G, B=@n6))
  )
  n8 = VectorParameter() {parameterName=EmissiveColorDark, defaultValue=<rgba 0 0.0417 0 0>}
  n9 = ComponentMask(None=TextureCoordinate()) {r=true}
  n10 = OneMinus(None=@n9)
  n11 = TextureSample(UVs=TextureCoordinate()) {texture=&T_Water_Top_Ramp}
  n12 = Multiply(A=@n10, B=@n11.G)
  n13 = LinearInterpolate(A=@n8.RGB, B=VectorParameter() {parameterName=EmissiveColorBright, defaultValue=<rgba 0.5772 6 0.5772 0>}.RGB, Alpha=@n12)
  n14 = ScalarParameter() {defaultValue=4, parameterName=ColorDepth}
  n15 = ShadingPathSwitch(Default=LinearInterpolate(A=Multiply(A=@n13) {constB=64}, B=@n13, Alpha=DepthFade(FadeDistance=@n14)), Mobile=Multiply(A=@n8.RGB, B=@n12)) {desc=Migrated from FeatureLevelSwitch}
  MaterialFunctionCall(Emissive=Multiply(A=@n7, B=Add(A=@n7, B=@n15))) {materialFunction=&MF_FX_EmissiveAdjustments}
  n16 = ScalarParameter() {defaultValue=0.2, parameterName=Non-Pulse_AlphaTweak}
  n17 = LinearInterpolate(Alpha=@n9) {constB=4}
  n18 = Multiply(A=Add(A=@n7, B=Multiply(A=@n11.G, B=@n16)), B=Multiply(A=Power(Base=@n10) {constExponent=3}, B=Clamp(None=@n17)))
  n19 = ScalarParameter() {defaultValue=1, parameterName=OpacityMultiply}
  Multiply(A=ParticleColor().A, B=Saturate(None=Multiply(A=@n18, B=@n19)))
 ]
}
